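# Physics asset for the Lionsmane vehicle body cosmetic, giving its skeletal mesh a simplified collision shape.
# A single body on the root bone holds one box, longest along X and raised above the bone origin to enclose the car as a whole.
# It has no other bodies and no constraints, so it defines no articulated or per-part physics.
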PhysicsAsset /VehicleCosmetics/Bodies/Lionsmane/Meshes/Body_Lionsmane_Physics_DM
bodies {root [{shapeType=Box, extentX=444, extentY=216, extentZ=132, shapeName=root_box, center=<16 0 105.1844>}]}
constraints []
